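# Masked lit material for the tentacle tendrils of the Clash skydive contrail, usable on Niagara mesh particles and skeletal meshes.
# View-space sine WPO wiggles them along UV1, offset by vertex color and particle parameters, masked along their length, with optional near-camera fade and Perlin jitter.
# An optional height-lerp cloud mask drives dissolve.
# Dynamic parameters scale brightness and switch packed-texture surface values against constants for frontend; panning noise adds normal detail.
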Material /BRCosmetics/Athena/Effects/Particles/Skydiving/NS_Contrail_Clash/M_Clash_SkyDive_Tendrils
graph {
 properties {
  allowTranslucentCustomDepthWrites=true
  bUsedWithMeshDeformer=true
  bUsedWithNiagaraMeshParticles=true
  bUsedWithSkeletalMesh=true
  blendMode=BLEND_Masked
  opacityMaskClipValue=0.05
  pixelDepthOffsetMode=PDOM_Legacy
  usedShadingModels=Default Lit
 }
 nodes [
  n1 = VectorParameter() {parameterName=WPO_Direction, defaultValue=<rgba 0.4 0 0 1>}
  n2 = ComponentMask(None=VectorParameter() {parameterName=UV_Scale, defaultValue=<rgba 1 1 0 1>}.RGB) {g=true, r=true}
  n3 = LinearInterpolate(B=Clamp(None=VertexColor().R) {maxDefault=1.5, minDefault=0.7}, Alpha=ScalarParameter() {parameterName=Use_UV_VertexColor}) {constA=1}
  n4 = ComponentMask(None=Multiply(A=TextureCoordinate() {coordinateIndex=1, uTiling=0, vTiling=0.7}, B=Multiply(A=@n2, B=@n3))) {g=true}
  n5 = ScalarParameter() {parameterName=WPO_Offset}
  n6 = Add(
   A=Add(
    A=Add(A=@n4, B=@n5)
    B=DynamicParameter() {defaultValue=<rgba 1 1 1 0>, paramNames [RandomOffset, WPO_Intensity, WPO_Speed, Dissolve]}.RandomOffset
   )
   B=VertexColor().R
  )
  n7 = Multiply(
   A=Time()
   B=Multiply(A=DynamicParameter() {defaultValue=<rgba 1 1 1 0>, paramNames [RandomOffset, WPO_Intensity, WPO_Speed, Dissolve]}.WPO_Speed)
  ) {constB=2.8}
  n8 = Multiply(A=Sine(None=Add(A=@n6, B=@n7))) {constB=20}
  n9 = ComponentMask(None=TextureCoordinate() {coordinateIndex=1}) {g=true}
  n10 = LinearInterpolate(Alpha=Add(A=OneMinus(None=@n9)) {constB=0.5})
  n11 = Transform(None=Multiply(A=Multiply(A=@n1.RGB, B=@n8), B=Saturate(None=@n10))) {transformSourceType=TRANSFORMSOURCE_View}
  n12 = ComponentMask(None=TextureCoordinate() {coordinateIndex=1}) {g=true}
  n13 = Power(
   Base=Add(A=OneMinus(None=@n12), B=ScalarParameter() {parameterName=Bottom_Mask_Offset})
   Exp=ScalarParameter() {defaultValue=2.9, parameterName=Power}
  )
  n14 = ScalarParameter() {defaultValue=50, parameterName=Power_Mask_Edge}
  n15 = Multiply(
   A=Multiply(
    A=Multiply(A=@n11) {constB=0.5}
    B=DynamicParameter() {defaultValue=<rgba 1 1 1 0>, paramNames [RandomOffset, WPO_Intensity, WPO_Speed, Dissolve]}.WPO_Intensity
   )
   B=Saturate(None=Multiply(A=@n13, B=@n14) {constB=50})
  )
  n16 = ScalarParameter() {defaultValue=0.3, group=Near_Camera_fade, parameterName=Min_Near_Camera_WPO}
  n17 = MaterialFunctionCall(Distance=ScalarParameter() {defaultValue=256, group=Near_Camera_fade, parameterName=Near_Camera_fade}) {materialFunction=&Near_Camera_Fade}
  n18 = LinearInterpolate(
   B=Saturate(None=LinearInterpolate(A=@n16, Alpha=@n17.Result) {constA=0.5})
   Alpha=ScalarParameter() {group=Near_Camera_fade, parameterName=Use_Near_Camera_Fade}
  ) {constA=1}
  n19 = TextureCoordinate() {coordinateIndex=1, uTiling=1.5, vTiling=1.5}
  n20 = AppendVector(A=Constant(), B=VertexColor().R)
  n21 = TextureSample(UVs=Panner(Coordinate=Add(A=@n19, B=@n20)) {speedX=0.075, speedY=0.3}) {samplerType=SAMPLERTYPE_Normal, texture=/BRCosmetics/Characters/Player/Male/Large/Bodies/M_LRG_Clash_V/Textures/Simple_Perlin_N}
  Add(
   A=LinearInterpolate(B=Multiply(A=@n15, B=@n18), Alpha=ScalarParameter() {defaultValue=1, parameterName=Use_WPO})
   B=Multiply(A=ScalarParameter() {parameterName=Small_WPO_Intensity}, B=@n21.RGB)
  ) {constB=0}
  n22 = TextureSample() {texture=&/T_Skydive_Clash_Tentacles_S}
  n23 = DynamicParameter() {parameterIndex=1, defaultValue=<rgba 1 1 1 0>, paramNames [Param1, Param2, Brightness_in_game, Use_Frontend_bpr]}
  LinearInterpolate(
   A=LinearInterpolate(A=Constant() {r=0.9}, B=Constant() {r=0.5}, Alpha=ScalarParameter() {parameterName=Use_Tendrils_bottom_Color})
   B=@n22.R
   Alpha=@n23.Use_Frontend_bpr
  )
  n24 = TextureCoordinate() {coordinateIndex=1, uTiling=2, vTiling=0.5}
  n25 = Multiply(A=MaterialFunctionCall!(
    Diffuse 1
    Diffuse 2
    Transistion Phase=DynamicParameter() {defaultValue=<rgba 1 1 1 0>, paramNames [RandomOffset, WPO_Intensity, WPO_Speed, Dissolve]}.Dissolve
    Height Texture 1
    Height Texture 2=TextureSample(UVs=@n24) {samplerType=SAMPLERTYPE_LinearGrayscale, texture=/Game/Athena/Prototype/Terrain/T_CloudCoverage}.R
    Contrast
   ) {materialFunction=/Engine/Functions/Engine_MaterialFunctions02/Texturing/HeightLerpWithTwoHeightMaps}.Alpha)
  n26 = ComponentMask(None=TextureCoordinate() {coordinateIndex=1}) {g=true}
  n27 = Multiply(A=Add(A=OneMinus(None=@n26)) {constB=-0.08}) {constB=20}
  LinearInterpolate(A=Multiply(A=Saturate(None=@n25), B=Saturate(None=@n27)), Alpha=ScalarParameter() {parameterName=Use_appearance_mask})
  LinearInterpolate(A=Constant() {r=0.3}, B=@n22.G, Alpha=@n23.Use_Frontend_bpr)
  LinearInterpolate(A=Constant() {r=0.4}, B=@n22.B, Alpha=@n23.Use_Frontend_bpr)
  n28 = LinearInterpolate(
   A=TextureSample() {texture=&/T_Skydive_Clash_Tentacles}.RGB
   B=TextureSampleParameter2D(UVs=TextureCoordinate() {coordinateIndex=1}) {parameterName=Tendril_Bottom, texture=&T_Skydive_Clash_Tentacles}.RGB
   Alpha=ScalarParameter() {parameterName=Use_Tendrils_bottom_Color}
  )
  n29 = DynamicParameter() {parameterIndex=1, defaultValue=<rgba 1 1 1 0>, paramNames [Param1, Param2, Brightness_in_game, Use_Frontend_bpr]}
  n30 = Multiply(A=VectorParameter() {parameterName=Color_Tweak, defaultValue=<rgba 0.8 0.6875 0.6988 1>}.RGB, B=@n29.Brightness_in_game)
  MaterialFunctionCall(Emissive=Multiply(A=Multiply(A=@n28) {constB=0.05}, B=@n30)) {materialFunction=&MF_FX_EmissiveAdjustments}
  Multiply(A=@n28, B=@n30)
  n31 = TextureSample() {samplerType=SAMPLERTYPE_Normal, texture=&/T_Skydive_Clash_Tentacles_N}
  n32 = Constant3Vector() {constant=<rgba 0 0 1 0>}
  n33 = ScalarParameter() {parameterName=Use_Tendrils_bottom_Color}
  n34 = Constant3Vector() {constant=<rgba 0 0 1 0>}
  n35 = ComponentMask(None=@n21.RGB) {g=true, r=true}
  n36 = Panner(Coordinate=Add(A=@n20, B=TextureCoordinate() {uTiling=2, vTiling=2})) {speedY=-0.03}
  n37 = Saturate(None=LinearInterpolate(Alpha=TextureSample(UVs=@n36) {texture=/Game/Effects/Fort_Effects/Textures/Power_Source/T_Power_Source_Sphere_Texture}.R) {constA=-0.2, constB=5})
  n38 = MaterialFunctionCall(Normal=AppendVector(A=Multiply(A=@n35, B=@n37), B=Constant() {r=1}), Flatness=ScalarParameter() {defaultValue=0.7, parameterName=Flatness}) {materialFunction=&FlattenNormal}
  n39 = ComponentMask(None=TextureCoordinate() {coordinateIndex=1}) {g=true}
  n40 = Saturate(None=Multiply(A=Power(Base=@n39)) {constB=10})
  Reroute(None=MaterialFunctionCall(BaseNormal=LinearInterpolate(A=@n31.RGB, B=@n32, Alpha=@n33), AdditionalNormal=LinearInterpolate(A=@n34, B=@n38.Result, Alpha=@n40)) {materialFunction=&BlendAngleCorrectedNormals}.Result)
  n41 = TextureCoordinate() {coordinateIndex=1}
  n42 = TextureSample(UVs=TextureCoordinate() {coordinateIndex=1, uTiling=2, vTiling=2}) {samplerType=SAMPLERTYPE_LinearColor, texture=/Game/Effects/Fort_Effects/BluGlow/Violentium/Violentium_Mask}
  Subtract(
   A=Add(A=ComponentMask(None=@n41) {g=true}, B=Multiply(A=@n42.R) {constB=0.3})
   B=LinearInterpolate(Alpha=MaterialFunctionCall() {materialFunction=&DebugTimeSine}.Result) {constA=-0.7, constB=1.5}
  ) {constB=-0.7}
 ]
}
